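# Additive, two-sided unlit material for Neo Tilted spotlight beams on Niagara mesh particles.
# Center-scaled UVs, lerped along V, shape a tapered soft-edged column, combined with two V-power gradients for lengthwise fade and brightness.
# Depth fade (skipped on Low quality and mobile, which adds a brightness multiplier), near-camera fade and particle alpha feed a TAA dither function.
# SplineThicken takes both widths from the WidthBase dynamic parameter, leaving WidthTip unused.
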Material /Game/Effects/Fort_Effects/Materials/NeoTilted/M_Neo_Tilted_Spotlight_Niagara
graph {
 properties {
  bUsedWithInstancedStaticMeshes=true
  bUsedWithNiagaraMeshParticles=true
  blendMode=BLEND_Additive
  pixelDepthOffsetMode=PDOM_Legacy
  shadingModel=MSM_Unlit
  twoSided=true
  usedShadingModels=Unlit
 }
 nodes [
  n1 = TextureCoordinate()
  n2 = ScalarParameter() {defaultValue=1.5, parameterName=Scale Tip}
  n3 = ScalarParameter() {defaultValue=0.2, parameterName=Scale Base}
  n4 = ComponentMask(None=@n1) {g=true}
  n5 = ComponentMask(None=MaterialFunctionCall(UVs=@n1, Texture Scale=LinearInterpolate(A=@n2, B=@n3, Alpha=@n4)) {materialFunction=&ScaleUVsByCenter}.UVs) {r=true}
  n6 = Multiply(A=@n5, B=OneMinus(None=@n5))
  n7 = ScalarParameter() {defaultValue=3, parameterName=Edge Fade Power}
  n8 = Saturate(None=Multiply(A=Power(Base=@n6, Exp=@n7) {constExponent=4}, B=ScalarParameter() {defaultValue=5, parameterName=Center Column Multiplier}))
  n9 = ComponentMask(None=TextureCoordinate()) {g=true}
  n10 = ScalarParameter() {defaultValue=1, parameterName=Tip Fade Power}
  n11 = Power(Base=@n9, Exp=ScalarParameter() {defaultValue=10, parameterName=Tip Fade Power2})
  n12 = ScalarParameter() {defaultValue=3, parameterName=Tip Fade Power2_Brightness}
  n13 = Add(A=Multiply(A=@n8, B=Power(Base=@n9, Exp=@n10)), B=Multiply(A=@n8, B=Multiply(A=@n11, B=@n12)))
  n14 = Multiply(A=@n13, B=DepthFade())
  n15 = QualitySwitch(Default=@n14, Low=@n13, High=@n14, Medium=@n14)
  n16 = ScalarParameter() {
   defaultValue=1
   desc=Default HDRMult is too bright for mobile. This allows to reduce for mobile when HDR is high.
   parameterName=MobileBrightnessMultiplier
  }
  n17 = ShadingPathSwitch(Default=@n15, Deferred=@n15, Mobile=Multiply(A=@n13, B=@n16)) {desc=Migrated from FeatureLevelSwitch}
  n18 = MaterialFunctionCall(Distance=ScalarParameter() {defaultValue=500, parameterName=NearFade}) {materialFunction=&Near_Camera_Fade}
  MaterialFunctionCall(Alpha Threshold=Multiply(A=ParticleColor().A, B=Multiply(A=@n17, B=@n18.Result))) {materialFunction=&DitherTemporalAA_Fortnite}
  n19 = DynamicParameter() {defaultValue=<rgba 1500 1 1 1>, paramNames [WidthBase,WidthTip,Param3,Param4]}
  MaterialFunctionCall(WidthBase=@n19.WidthBase, WidthTip=@n19.WidthBase) {desc=Can't be done in instances, materialFunction=&SplineThicken}
  ParticleColor()
 ]
}
