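# Translucent, unlit, two-sided material for Neo Tilted spotlight beams, with optional time-of-day tinting and daytime shutoff.
# World position offset builds a camera-facing beam from UVs and sways it around a pivot with a time-driven sine, with a Creative switch for local-space axes.
# Opacity combines tapered column and tip fades, cloud samples whose world-position UVs are zeroed, and near-camera fade.
# Depth fade is skipped on low quality and mobile, where a brightness multiplier applies.
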Material /Game/Effects/Fort_Effects/Materials/NeoTilted/M_Neo_Tilted_Spotlight_TwoSided
graph {
 properties {
  bUsedWithInstancedStaticMeshes=true
  blendMode=BLEND_Translucent
  pixelDepthOffsetMode=PDOM_Legacy
  shadingModel=MSM_Unlit
  twoSided=true
  usedShadingModels=Unlit
 }
 nodes [
  n1 = Constant3Vector() {constant=<rgba 0 0 1 0>}
  n2 = ScalarParameter() {parameterName=rotation}
  n3 = Multiply(A=Constant() {r=34.9111}, B=Frac(None=Max(A=@n2) {constB=0})) {constA=16}
  n4 = Constant3Vector()
  n5 = VectorParameter() {parameterName=Rotation Vector, defaultValue=<rgba 1 0 0 1>}
  n6 = StaticSwitchParameter(
   True=Transform(None=Constant3Vector() {constant=<rgba 0 0 -1 0>}) {transformSourceType=TRANSFORMSOURCE_Local}
   False=Normalize(VectorInput=RotateAboutAxis(NormalizedRotationAxis=@n1, RotationAngle=@n3, PivotPoint=@n4, Position=@n5.RGB))
  ) {group=Creative Switch, parameterName=Is Creative, sortPriority=0}
  n7 = Time()
  n8 = LinearInterpolate(
   A=ScalarParameter() {defaultValue=0.125, parameterName=Rotation Speed}
   B=ScalarParameter() {defaultValue=0.07, parameterName=Rotation Speed2}
   Alpha=Frac(None=@n3)
  )
  n9 = Multiply(A=Sine(None=Multiply(A=@n7, B=@n8)), B=ScalarParameter() {defaultValue=0.125, parameterName=Rotation Max})
  n10 = VectorParameter() {parameterName=Pivot Offset Custom, defaultValue=<rgba 0 0 75 1>}
  n11 = Add(
   A=ActorPositionWS()
   B=StaticSwitchParameter(True=Transform(None=@n10.RGB) {transformSourceType=TRANSFORMSOURCE_Local}, False=@n10.RGB) {group=Creative Switch, parameterName=Is Creative, sortPriority=0}
  )
  n12 = Transform(None=Constant3Vector() {constant=<rgba 1 0 0 0>}) {transformSourceType=TRANSFORMSOURCE_Local}
  n13 = StaticSwitchParameter(True=@n12, False=Constant3Vector() {constant=<rgba 0 0 1 0>}) {group=Creative Switch, parameterName=Is Creative, sortPriority=0}
  n14 = OneMinus(None=ComponentMask(None=TextureCoordinate()) {g=true})
  n15 = ScalarParameter() {defaultValue=15, parameterName=ScaleHeight}
  n16 = CollectionParameter() {
   collection=&FortniteMaterialParameters
   parameterName=FortniteActiveTimeOfDay
  }
  n17 = DotProduct(A=@n16, B=Constant4Vector() {constant=<rgba 0 0 0 1>}.RGBA)
  n18 = StaticSwitchParameter(True=Multiply(A=@n15, B=@n17), False=@n15) {parameterName=Turn Off At Day}
  n19 = CrossProduct(
   A=MaterialFunctionCall() {materialFunction=&CameraDirectionVector}.Result
   B=Transform(None=Constant3Vector() {constant=<rgba 1 0 0 0>}) {transformSourceType=TRANSFORMSOURCE_Local}
  )
  n20 = Constant() {r=-0.5}
  n21 = TextureCoordinate()
  n22 = ScalarParameter() {parameterName=Time Falloff Power}
  n23 = ScalarParameter() {defaultValue=1500, parameterName=Main Width}
  n24 = StaticSwitchParameter(True=Multiply(A=Power(Base=@n17, Exp=@n22), B=@n23), False=@n23) {parameterName=Turn Off At Day}
  n25 = Multiply(A=ComponentMask(None=Add(A=@n20, B=@n21)) {r=true}, B=@n24)
  n26 = AppendVector(A=@n25, B=@n25)
  n27 = Constant()
  n28 = VectorParameter() {parameterName=fdfd, defaultValue=<rgba 1 0 0 1>}
  n29 = StaticSwitchParameter(
   True=Multiply(A=Normalize(VectorInput=@n19), B=@n25)
   False=Multiply(A=AppendVector(A=@n26, B=@n27), B=Transform(None=@n28.RGB) {transformSourceType=TRANSFORMSOURCE_Camera})
  ) {group=Creative Switch, parameterName=Is Creative, sortPriority=0}
  n30 = Add(A=@n29, B=Multiply(A=Multiply(A=@n14, B=@n18), B=StaticSwitchParameter(True=@n13, False=@n12) {defaultValue=true, parameterName=Rotate}))
  n31 = Add(A=@n30, B=WorldPosition().XYZ)
  Add(
   A=@n30
   B=StaticSwitchParameter(True=RotateAboutAxis(NormalizedRotationAxis=@n6, RotationAngle=@n9, PivotPoint=@n11, Position=@n31), False=Constant()) {defaultValue=true, parameterName=Rotate}
  )
  DepthFade()
  LinearInterpolate(
   A=@n29
   B=MaterialFunctionCall(WidthBase=@n24, WidthTip=@n24) {desc=Can't be done in instances, materialFunction=&SplineThicken}.WorldPositionOffset
   Alpha=ScalarParameter() {parameterName=DebugTestImposterThicken}
  )
  MaterialFunctionCall() {desc="No longer supported, HLODs don't bake positions to UVs", materialFunction=/Game/ImpostorBaker/MaterialFunctions/ImpostorInfo_FromUVs}
  n32 = TextureCoordinate()
  n33 = ScalarParameter() {defaultValue=1.5, parameterName=Scale Tip}
  n34 = ScalarParameter() {defaultValue=0.2, parameterName=Scale Base}
  n35 = ComponentMask(None=@n32) {g=true}
  n36 = ComponentMask(None=MaterialFunctionCall(UVs=@n32, Texture Scale=LinearInterpolate(A=@n33, B=@n34, Alpha=@n35)) {materialFunction=&ScaleUVsByCenter}.UVs) {r=true}
  n37 = Multiply(A=@n36, B=OneMinus(None=@n36))
  n38 = ScalarParameter() {defaultValue=3, parameterName=Edge Fade Power}
  n39 = Saturate(None=Multiply(A=Power(Base=@n37, Exp=@n38) {constExponent=4}, B=ScalarParameter() {defaultValue=5, parameterName=Center Column Multiplier}))
  n40 = ComponentMask(None=TextureCoordinate()) {g=true}
  n41 = Power(Base=@n40, Exp=ScalarParameter() {defaultValue=1, parameterName=Tip Fade Power})
  n42 = Multiply(
   A=Power(Base=@n40, Exp=ScalarParameter() {defaultValue=10, parameterName=Tip Fade Power2})
   B=ScalarParameter() {defaultValue=3, parameterName=Tip Fade Power2_Brightness}
  )
  n43 = WorldPosition()
  n44 = Constant() {r=0}
  n45 = TextureSample!(UVs=ComponentMask(None=Multiply(A=@n43.XYZ, B=@n44)) {g=true, r=true}, Tex, Bias, Apply View MipBias) {
   mipValueMode=TMVM_MipBias
   samplerSource=SSM_Wrap_WorldGroupSettings
   samplerType=SAMPLERTYPE_LinearColor
   texture=/Game/Global/Textures/Default/Tileable_Patterns/Cloud_Diffuse
  }
  n46 = ComponentMask(None=Multiply(A=WorldPosition().XYZ, B=Multiply(A=@n44) {constB=1.185})) {b=true, r=true}
  n47 = Multiply(A=Multiply(
    A=@n45.G
    B=TextureSample!(UVs=@n46, Tex, Apply View MipBias) {samplerType=SAMPLERTYPE_LinearColor, texture=&Cloud_Diffuse}.G
   ))
  n48 = ScalarParameter() {defaultValue=0.1, parameterName=Thicken Beam}
  n49 = Multiply(A=Add(A=Multiply(A=@n39, B=@n41), B=Multiply(A=@n39, B=@n42)), B=Saturate(None=Add(A=@n47, B=@n48)))
  n50 = ScalarParameter() {defaultValue=100, parameterName=Depth_FadeDistance_Top}
  n51 = ScalarParameter() {defaultValue=100, parameterName=Depth_FadeDistance_Bottom}
  n52 = ComponentMask(None=TextureCoordinate()) {g=true}
  n53 = Multiply(A=@n49, B=DepthFade(FadeDistance=LinearInterpolate(A=@n50, B=@n51, Alpha=@n52)))
  n54 = QualitySwitch(Default=@n53, Low=@n49, High=@n53, Medium=@n53)
  n55 = ScalarParameter() {
   defaultValue=1
   desc=Default HDRMult is too bright for mobile. This allows to reduce for mobile when HDR is high.
   parameterName=MobileBrightnessMultiplier
  }
  n56 = Multiply(A=@n49, B=@n55)
  n57 = Multiply(
   A=FeatureLevelSwitch(Default=@n54, ES2=@n56, ES3_1=@n56, SM4_REMOVED=@n54, SM5=@n54, SM6=@n54)
   B=MaterialFunctionCall(Distance=ScalarParameter() {defaultValue=500, parameterName=NearFade}) {materialFunction=&Near_Camera_Fade}.Result
  )
  n58 = CollectionParameter() {
   collection=&FortniteMaterialParameters
   parameterName=FortniteActiveTimeOfDay
  }
  n59 = Constant4Vector() {constant=<rgba 0 0 0 1>}
  StaticSwitchParameter(True=LinearInterpolate(B=@n57, Alpha=DotProduct(A=@n58, B=@n59.RGBA)), False=@n57) {parameterName=Turn Off At Day}
  StaticSwitchParameter() {parameterName=UseDepthFade?}
  TransformPosition(None=Constant3Vector()) {transformType=TRANSFORMPOSSOURCE_World}
  n60 = MaterialFunctionCall(
   Value_Morning=ScalarParameter() {defaultValue=1, parameterName="Value_Morning (S)"}
   Value_Day=ScalarParameter() {defaultValue=1, parameterName="Value_Day (S)"}
   Value_Evening=ScalarParameter() {defaultValue=1, parameterName="Value_Evening (S)"}
   Value_Night=ScalarParameter() {defaultValue=1, parameterName="Value_Night (S)"}
  ) {materialFunction=&Scalar_Time_of_Day}
  n61 = Multiply(
   A=LinearInterpolate(
    A=Constant() {r=1}
    B=MaterialFunctionCall() {materialFunction=/Game/Effects/Fort_Effects/Materials/_Functions/CurveColor_Time_of_Day}.Result
    Alpha=ScalarParameter() {defaultValue=1, parameterName=TimeofDayColor_Amount}
   )
   B=VectorParameter() {parameterName=Color, defaultValue=<rgba 3.7163 1.2205 4 1>}.RGB
  )
  VertexInterpolator(VS=StaticSwitchParameter(True=Multiply(A=@n60.Result, B=@n61), False=@n61) {parameterName=UseTOD_Intensity?})
 ]
}
