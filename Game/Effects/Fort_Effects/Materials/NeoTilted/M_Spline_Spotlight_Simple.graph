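# Additive, unlit, two-sided material for a spotlight beam effect built on SplineThicken.
# Emissive color combines particle color and a tint, scaled by time-of-day brightness.
# A UV-tapered center column with tip-weighted falloff is multiplied by depth fade, near-camera fade, particle alpha and a view-angle fade from the SplineThicken normal.
# Quality and mobile switches substitute simpler depth fade and optional low-end visibility boosts.
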Material /Game/Effects/Fort_Effects/Materials/NeoTilted/M_Spline_Spotlight_Simple
graph {
 properties {
  bUsedWithInstancedStaticMeshes=true
  blendMode=BLEND_Additive
  pixelDepthOffsetMode=PDOM_Legacy
  shadingModel=MSM_Unlit
  twoSided=true
  usedShadingModels=Unlit
 }
 nodes [
  n1 = ParticleColor()
  n2 = Multiply(A=VectorParameter() {parameterName=Color, defaultValue=<rgba 0.35 0.1953 0.0192 0>}.RGB) {constB=5}
  n3 = Multiply(A=@n2, B=Constant() {r=1.5})
  n4 = LinearInterpolate(A=@n2, B=QualitySwitch(Default=@n2, Low=@n3, Medium=@n3), Alpha=ScalarParameter() {parameterName=VisibilityFixOnLow})
  n5 = ScalarParameter() {defaultValue=1, parameterName=DayBrightness}
  n6 = ScalarParameter() {defaultValue=0.7, parameterName=EveningBrightness}
  n7 = ScalarParameter() {defaultValue=0.5, parameterName=NightBrightness}
  MaterialFunctionCall(Emissive=Multiply(
    A=Multiply(A=@n1.RGB, B=@n4)
    B=MaterialFunctionCall(Color_Morning=@n5, Color_Day=@n5, Color_Evening=@n6, Color_Night=@n7) {materialFunction=&Color_Time_of_Day}.Result
   )) {materialFunction=&MF_FX_EmissiveAdjustments}
  Multiply(A=@n2, B=Constant() {r=1.5})
  n8 = TextureCoordinate()
  n9 = ScalarParameter() {defaultValue=1, parameterName=Scale Tip}
  n10 = ScalarParameter() {defaultValue=1, parameterName=Scale Base}
  n11 = ComponentMask(None=@n8) {g=true}
  n12 = ComponentMask(None=MaterialFunctionCall(UVs=@n8, Texture Scale=LinearInterpolate(A=@n9, B=@n10, Alpha=@n11)) {materialFunction=&ScaleUVsByCenter}.UVs) {r=true}
  n13 = Multiply(A=@n12, B=OneMinus(None=@n12))
  n14 = ScalarParameter() {defaultValue=3, parameterName=Edge Fade Power}
  n15 = Saturate(None=Multiply(A=Power(Base=@n13, Exp=@n14) {constExponent=4}, B=ScalarParameter() {defaultValue=5, parameterName=Center Column Multiplier}))
  n16 = ComponentMask(None=TextureCoordinate()) {g=true}
  n17 = ScalarParameter() {defaultValue=1.25, parameterName=Tip Fade Power}
  n18 = Power(Base=@n16, Exp=ScalarParameter() {defaultValue=10, parameterName=Tip Fade Power2})
  n19 = ScalarParameter() {defaultValue=1, parameterName=Tip Fade Power2_Brightness}
  n20 = Add(A=Multiply(A=@n15, B=Power(Base=@n16, Exp=@n17)), B=Multiply(A=@n15, B=Multiply(A=@n18, B=@n19)))
  n21 = LinearInterpolate(
   A=ScalarParameter() {defaultValue=200, parameterName=EndDepthFade}
   B=ScalarParameter() {defaultValue=10, parameterName=StartDepthFade}
   Alpha=ComponentMask(None=TextureCoordinate()) {g=true}
  )
  n22 = Multiply(
   A=ShadingPathSwitch(Default=DepthFade(FadeDistance=@n21), Mobile=Constant() {r=0.8})
   B=MaterialFunctionCall(Distance=ScalarParameter() {defaultValue=400, parameterName=NearFade}) {materialFunction=&Near_Camera_Fade}.Result
  )
  n23 = DotProduct(
   A=CameraVectorWS()
   B=MaterialFunctionCall(
    WidthBase=ScalarParameter() {defaultValue=5, parameterName=WidthBase}
    WidthTip=ScalarParameter() {defaultValue=800, parameterName=WidthTipMult}
   ) {materialFunction=&SplineThicken}.Normal
  )
  n24 = ScalarParameter() {defaultValue=3, parameterName=AngleFadePower}
  n25 = QualitySwitch(Default=@n24, Low=Constant() {r=2.2})
  n26 = ScalarParameter() {parameterName=VisibilityFixOnLow}
  n27 = Multiply(A=Power(Base=Abs(None=@n23) {desc=Make this work on two-sided materials}, Exp=LinearInterpolate(A=@n24, B=@n25, Alpha=@n26))) {constB=2}
  Multiply(A=@n1.A, B=Multiply(A=Multiply(A=@n20, B=@n22), B=Saturate(None=@n27)))
 ]
}
